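; Weapon actor Blueprint for the Nitro Gauntlets, linking damage events, punch preview and ammo state to the owning pawn.
; On equip, authority forwards damage-dealt events to the UI via RepDamageDealt, flagging non-Pawn, non-Vehicle hits as environment.
; Remote clients run an optional looping punch-preview timer, cleared on unequip, and check initial ammo.
; CheckAmmoServer applies or removes an out-of-ammo GameplayEffect so Mantis knows when to aerial attack.
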(event EventBeginPlay
  ;[0 1 0.4077] Begin play
  (bind _asfort_pawn (Utilities|Casting|CastToFortPawn (Game|GetInstigator)))
  (Variables|Default|SetPlayerPawn _asfort_pawn))

(event FortWeapon|EventOnEquippedAndAttached
  ;[0 1 0.4077] On equipped
  (Networking|SwitchHasAuthority
    (:Authority
      ;[1 0.3767 0] Replicate damage dealt to UI
      (bind (_asynctaskproxy _payload) (Ability|Async|WaitGameplayEventToActor (Variables|Default|GetPlayerPawn) "(TagName=\"Event.Damage.Dealt\")")
        (:EventReceived
          (bind _returnvalue (GameplayTags|HasAnyMatchingGameplayTags(Message) (.Target (Utilities|Struct|BreakGameplayEventData _payload)) (GameplayTags|MakeLiteralGameplayTagContainer "(GameplayTags=((TagName=\"Pawn\"),(TagName=\"Vehicle\")))")))
          (CallFunction|RepDamageDealt (.EventMagnitude (Utilities|Struct|BreakGameplayEventData _payload)) (not _returnvalue)))))
    (:Remote
      (if (Variables|Default|GetbPunchPreviewEnabled)
        (bind _returnvalue_2 (Utilities|Time|SetTimerbyEvent (EventDispatchers|CreateEvent :!event UpdatePunchPreviewNative) (Variables|Default|GetPunchPreviewRate) true true))
        (Variables|Default|SetPunchPreviewTimer _returnvalue_2)
        (else
          _))
      (CallFunction|InitAmmoCheck))))

(event Custom|RepDamageDealt
  (params (Damage :type Float%20%28double-precision%29) (EnviroHit? :type Boolean))
  ;[1 0.3767 0] Replicate damage dealt to UI
  (Default|CallOnDidDamage Damage EnviroHit?))

(event EventOnAmmoCountChanged
  ;[0 1 0.4077] On ammo changed
  (CallFunction|CheckAmmoServer (<= (FortWeapon|GetRemainingAmmo) 0)))

(event FortWeapon|EventOnUnEquippedAndDetached
  ;[0 1 0.4077] End update timer on unequip
  (Networking|CanExecuteCosmeticEvents
    (:True
      (Utilities|Time|ClearandInvalidateTimerbyHandle (Variables|Default|GetPunchPreviewTimer))
      (NitroGauntlets|CleanupAsyncTargeting))))

(event Custom|CheckAmmoServer
  (params (Out? :type Boolean))
  (bind _playerpawn (Variables|Default|GetPlayerPawn))
  (bind _returnvalue (Ability|GetFortAbilitySystemComponent _playerpawn))
  ;[0 1 0.4077] Add no ammo tags when we have no ammo so Mantis knows when to do aerial attack
  (if Out?
    (bind _returnvalue_2 (GameplayEffects|ApplyGameplayEffectToTarget :self _returnvalue :GameplayEffectClass "/NitroGauntletsGameplay/Gameplay/Effects/GE_Moonflax_NitroGauntlets_OutOfAmmo.GE_Moonflax_NitroGauntlets_OutOfAmmo_C" :Target _returnvalue :Level 0.0))
    (Variables|Default|SetOutOfAmmoSpec _returnvalue_2)
    (else
      (GameplayEffects|RemoveActiveGameplayEffect _returnvalue (Variables|Default|GetOutOfAmmoSpec) -1)
      (GameplayEffects|RemoveActiveEffectswithGrantedTags _returnvalue "(GameplayTags=((TagName=\"Granted.Moonflax.NitroGauntlets.NoAmmo\")))") ; Extra check for when weapon is dropped and picked up again
    )))

(event Custom|InitAmmoCheck
  ;[0 1 0.4077] Check ammo when weapon is first picked up
  (Networking|SwitchHasAuthority
    (:Remote
      (CallFunction|CheckAmmoServer (<= (FortWeapon|GetLocalRemainingAmmo) 0)))))

(fn ConstructionScript
  (|CallParentFunction))
